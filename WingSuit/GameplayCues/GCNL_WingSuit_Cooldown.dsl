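; Looping gameplay cue notify for the WingSuit cooldown, providing its visual and audio feedback on the player pawn.
; On loop start it snaps the cue's particle components to the Backpack mesh's FX_Idle socket.
; For the locally viewed pawn only, it spawns a looping cooldown sound on the current weapon, calls that weapon's cooldown audio trigger and fades out its warm-up audio.
; On removal it calls the parent and fades out the stored loop.
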(event EventBeginPlay)

(event Collision|EventActorBeginOverlap
  (params (OtherActor :type Actor%20Object%20Reference)))

(event EventTick
  (params (DeltaSeconds :type Float%20%28single-precision%29)))

(event EventOnLoopingStart(EitherVFX-preferthisversion)
  (params (MyTarget :type Actor%20Object%20Reference) (Parameters :type Gameplay%20Cue%20Parameters%20Structure%20%28by%20ref%29) (ParticleComponents :type Array%20of%20FXSystem%20Component%20Object%20References) (AudioComponents :type Array%20of%20Audio%20Component%20Object%20References))
  (bind _asfort_player_pawn (Utilities|Casting|CastToFortPlayerPawn MyTarget))
  (for _array_element ParticleComponents
    (Transformation|AttachComponentToComponent _array_element (Customization|GetSkeletalMeshforPartType _asfort_player_pawn "Backpack" :!class /Script/FortniteGame.FortPlayerPawn) "FX_Idle" "SnapToTarget" "KeepRelative" "SnapToTarget" true))
  (if (Pawn|IsLocallyViewed _asfort_player_pawn)
    (bind _returnvalue (Audio|SpawnSoundAttached "/WingSuit/Audio/MSS_WingsuitCoolDown_Loop.MSS_WingsuitCoolDown_Loop" (Variables|/Script/Engine.Actor|GetRootComponent (Variables|/Script/FortniteGame.FortPawn|GetCurrentWeapon _asfort_player_pawn))))
    (Variables|Default|SetLoopingAudioComp _returnvalue)
    (bind _asb_weap_wing_suit_physics_flying (Utilities|Casting|CastToB_Weap_WingSuit_PhysicsFlying (Variables|/Script/FortniteGame.FortPawn|GetCurrentWeapon _asfort_player_pawn)))
    (Class|BWeapWingSuitPhysicsFlying|TriggerCooldownAudio _asb_weap_wing_suit_physics_flying)
    (bind _asb_weap_wing_suit_physics_flying_2 (Utilities|Casting|CastToB_Weap_WingSuit_PhysicsFlying (Variables|/Script/FortniteGame.FortPawn|GetCurrentWeapon _asfort_player_pawn)))
    (Utilities|IsValid (Variables|/WingSuit/Gameplay/WingSuit_PhysicsFlying/B_Weap_WingSuit_PhysicsFlying.B_Weap_WingSuit_PhysicsFlying_C|GetWarmUpAudio _asb_weap_wing_suit_physics_flying_2)
      (:Is%20Valid
        (Audio|Components|Audio|FadeOut (Variables|/WingSuit/Gameplay/WingSuit_PhysicsFlying/B_Weap_WingSuit_PhysicsFlying.B_Weap_WingSuit_PhysicsFlying_C|GetWarmUpAudio _asb_weap_wing_suit_physics_flying_2) 0.25 0.0 "Linear" :!class /Script/Engine.AudioComponent)))))

(event EventOnRemoval(EitherVFX-preferthisversion)
  (params (MyTarget :type Actor%20Object%20Reference) (Parameters :type Gameplay%20Cue%20Parameters%20Structure%20%28by%20ref%29) (ParticleComponents :type Array%20of%20FXSystem%20Component%20Object%20References) (AudioComponents :type Array%20of%20Audio%20Component%20Object%20References) (BurstCameraShakeInstance :type Camera%20Shake%20Base%20Object%20Reference) (BurstDecalInstance :type Decal%20Actor%20Object%20Reference))
  (|CallParentFunction MyTarget Parameters ParticleComponents AudioComponents BurstCameraShakeInstance BurstDecalInstance)
  (bind _loopingaudiocomp (Variables|Default|GetLoopingAudioComp)
    (:then
      (Audio|Components|Audio|FadeOut _loopingaudiocomp 0.25 0.0 "Linear" :!class /Script/Engine.AudioComponent))
    (:else)))

(fn ConstructionScript)
