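; Looping gameplay cue notify for the WingSuit flare effect that colors its particles to match the suit's smoke screen color.
; On looping start, it casts the target to a Fort player pawn and its current weapon to B_Weap_WingSuit_PhysicsFlying.
; It then writes that weapon's GetSmokeScreenColor value into the "ColorIndex" int parameter of every spawned particle component.
; The other events, the OnRep_ColorIndex function and the construction script are empty.
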(event EventBeginPlay)

(event Collision|EventActorBeginOverlap
  (params (OtherActor :type Actor%20Object%20Reference)))

(event EventTick
  (params (DeltaSeconds :type Float%20%28single-precision%29)))

(event EventOnLoopingStart(EitherVFX-preferthisversion)
  (params (MyTarget :type Actor%20Object%20Reference) (Parameters :type Gameplay%20Cue%20Parameters%20Structure%20%28by%20ref%29) (ParticleComponents :type Array%20of%20FXSystem%20Component%20Object%20References) (AudioComponents :type Array%20of%20Audio%20Component%20Object%20References))
  (bind _asfort_player_pawn (Utilities|Casting|CastToFortPlayerPawn MyTarget))
  (bind _asb_weap_wing_suit_physics_flying (Utilities|Casting|CastToB_Weap_WingSuit_PhysicsFlying (Weapon|GetCurrentWeapon _asfort_player_pawn)))
  (for _array_element ParticleComponents
    (Effects|Components|ParticleSystem|SetIntParameter _array_element "ColorIndex" (Variables|/WingSuit/Gameplay/WingSuit_PhysicsFlying/B_Weap_WingSuit_PhysicsFlying.B_Weap_WingSuit_PhysicsFlying_C|GetSmokeScreenColor _asb_weap_wing_suit_physics_flying))))

(fn OnRep_ColorIndex)

(fn ConstructionScript)
