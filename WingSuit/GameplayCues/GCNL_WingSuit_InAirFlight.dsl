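; Looping GameplayCue notify that handles audio for a player's in-air wingsuit flight.
; On looping start it attaches a first- or third-person loop sound based on local view, then finds or adds a BP_Wingsuit_AudioController component on the pawn.
; For pawns not on the local player's team it also starts a looping sound indicator using a temporary MistRaven fire cue.
; Removal fades out the loop, stops the indicator and stops the controller's audio.
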(event EventBeginPlay)

(event Collision|EventActorBeginOverlap
  (params (OtherActor :type Actor%20Object%20Reference)))

(event EventTick
  (params (DeltaSeconds :type Float%20%28single-precision%29)))

(event EventOnLoopingStart(EitherVFX-preferthisversion)
  (params (MyTarget :type Actor%20Object%20Reference) (Parameters :type Gameplay%20Cue%20Parameters%20Structure%20%28by%20ref%29) (ParticleComponents :type Array%20of%20FXSystem%20Component%20Object%20References) (AudioComponents :type Array%20of%20Audio%20Component%20Object%20References))
  (bind _asfort_player_pawn_athena (Utilities|Casting|CastToFortPlayerPawnAthena MyTarget))
  (bind _returnvalue (Audio|SpawnSoundAttached (select (Pawn|IsLocallyViewed _asfort_player_pawn_athena) (Variables|Default|GetGrimStrutLoopSound_1P) (Variables|Default|GetGrimStrutLoopSound_3P)) (Variables|/Script/Engine.Actor|GetRootComponent _asfort_player_pawn_athena)))
  (Variables|Default|SetGrimStrutAudio _returnvalue)
  (CallFunction|AddWingsuitAudioController _asfort_player_pawn_athena)
  (if (not (Team|OnLocalPlayersTeam _asfort_player_pawn_athena _asfort_player_pawn_athena))
    (SoundIndicator|StartTargetedSoundIndicatorLooping (Variables|Default|GetFortSoundIndicator) _asfort_player_pawn_athena "/MistRaven/Assets/Audio/Temp_AudioVisualizer/MistRaven_Temp_Fire_Cue.MistRaven_Temp_Fire_Cue" 0.1 (Variables|/Script/Engine.Pawn|GetPlayerState _asfort_player_pawn_athena))))

(event EventOnRemoval(EitherVFX-preferthisversion)
  (params (MyTarget :type Actor%20Object%20Reference) (Parameters :type Gameplay%20Cue%20Parameters%20Structure%20%28by%20ref%29) (ParticleComponents :type Array%20of%20FXSystem%20Component%20Object%20References) (AudioComponents :type Array%20of%20Audio%20Component%20Object%20References) (BurstCameraShakeInstance :type Camera%20Shake%20Base%20Object%20Reference) (BurstDecalInstance :type Decal%20Actor%20Object%20Reference))
  (Audio|Components|Audio|FadeOut (Variables|Default|GetGrimStrutAudio) 0.2 0.0 "Linear" :!class /Script/Engine.AudioComponent)
  (SoundIndicator|StopSoundIndicatorLooping (Variables|Default|GetFortSoundIndicator))
  (bind _movementaudiocontroller (Variables|Default|GetMovementAudioController)
    (:then
      (Class|BPWingsuitAudioController|StopAudio _movementaudiocontroller))
    (:else)))

(fn AddWingsuitAudioController
  (params (PawnRef :type Actor%20Object%20Reference))
  (bind _returnvalue (Actor|GetComponentbyClass PawnRef "/WingSuit/Audio/BP_Wingsuit_AudioController.BP_Wingsuit_AudioController_C"))
  (Utilities|IsValid _returnvalue
    (:Is%20Valid
      (Variables|Default|SetMovementAudioController _returnvalue))
    (:Is%20Not%20Valid
      (bind _returnvalue_2 (Game|AddComponentbyClass PawnRef PawnRef "/WingSuit/Audio/BP_Wingsuit_AudioController.BP_Wingsuit_AudioController_C"))
      (Variables|Default|SetMovementAudioController _returnvalue_2))))

(fn ConstructionScript)
